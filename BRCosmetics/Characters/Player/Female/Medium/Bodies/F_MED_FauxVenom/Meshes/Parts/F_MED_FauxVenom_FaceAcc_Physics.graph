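# Physics asset for the FauxVenom female medium character's face accessory mesh, defining collision and joints for dangling ear chains and front and top "locane" strands.
# Small capsules form each chain, with boxes on the ear-chain ends and larger capsules for head and neck.
# Swing-limited constraints link chains segment by segment to the head, with ear twist locked and asymmetric top strands.
# The head-to-neck joint leaves all angular motion free.
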PhysicsAsset /BRCosmetics/Characters/Player/Female/Medium/Bodies/F_MED_FauxVenom/Meshes/Parts/F_MED_FauxVenom_FaceAcc_Physics
bodies {
 dyn_ear_l_chain_01 [{shapeType=Capsule, length=1, radius=0.5, shapeName=dyn_ear_l_chain_01_capsule, rotation=<rot 90 0 0>}]
 dyn_ear_l_chain_02 [{shapeType=Capsule, length=1, radius=0.5, shapeName=dyn_ear_l_chain_02_capsule, rotation=<rot 90 0 0>}]
 dyn_ear_l_chain_03 [{shapeType=Box, extentX=3.8519, extentY=3.8587, extentZ=0.7223, shapeName=dyn_ear_l_chain_03_box, center=<1.3744 0 0>}]
 dyn_ear_r_chain_01 [{shapeType=Capsule, length=1, radius=0.5, shapeName=dyn_ear_r_chain_01_capsule, rotation=<rot 90 0 0>}]
 dyn_ear_r_chain_02 [{shapeType=Capsule, length=1, radius=0.5, shapeName=dyn_ear_r_chain_02_capsule, rotation=<rot 90 0 0>}]
 dyn_ear_r_chain_03 [{shapeType=Box, extentX=3.8871, extentY=3.8558, extentZ=0.54, shapeName=dyn_ear_r_chain_03_box, center=<1.3505 0 0>}]
 dyn_front_locane_l_01 [{shapeType=Capsule, length=3.0162, radius=0.7261, shapeName=dyn_front_locane_l_01_capsule, center=<-0.136 0 0>, rotation=<rot 90 0 0>}]
 dyn_front_locane_l_02 [{shapeType=Capsule, length=3.4861, radius=0.6574, shapeName=dyn_front_locane_l_02_capsule, rotation=<rot 90 0 0>}]
 dyn_front_locane_r_01 [{shapeType=Capsule, length=2.9488, radius=0.6771, shapeName=dyn_front_locane_r_01_capsule, center=<0.3445 0 0>, rotation=<rot 90 0 0>}]
 dyn_front_locane_r_02 [{shapeType=Capsule, length=3.0201, radius=0.656, shapeName=dyn_front_locane_r_02_capsule, center=<0.5621 0 0>, rotation=<rot 80 -0.0001 0>}]
 dyn_top_locane_d_l_01 [{shapeType=Capsule, length=1, radius=0.5, shapeName=dyn_top_locane_d_l_01_capsule, rotation=<rot 90 0 0>}]
 dyn_top_locane_d_l_02 [{shapeType=Capsule, length=1, radius=0.5, shapeName=dyn_top_locane_d_l_02_capsule, rotation=<rot 90 0 0>}]
 dyn_top_locane_d_r_01 [{shapeType=Capsule, length=1, radius=0.5, shapeName=dyn_top_locane_d_r_01_capsule, rotation=<rot 90 0 0>}]
 dyn_top_locane_d_r_02 [{shapeType=Capsule, length=1, radius=0.5, shapeName=dyn_top_locane_d_r_02_capsule, rotation=<rot 90 0 0>}]
 dyn_top_locane_t_l_01 [{shapeType=Capsule, length=0.9985, radius=0.7736, shapeName=dyn_top_locane_t_l_01_capsule, rotation=<rot 90 0 0>}]
 dyn_top_locane_t_l_02 [{shapeType=Capsule, length=0.9985, radius=0.7736, shapeName=dyn_top_locane_t_l_02_capsule, rotation=<rot 90 0 0>}]
 dyn_top_locane_t_l_03 [{shapeType=Capsule, length=0.9985, radius=0.7736, shapeName=dyn_top_locane_t_l_03_capsule, rotation=<rot 90 0 0>}]
 dyn_top_locane_t_r_01 [{shapeType=Capsule, length=1, radius=0.5, shapeName=dyn_top_locane_t_r_01_capsule, rotation=<rot 90 0 0>}]
 dyn_top_locane_t_r_02 [{shapeType=Capsule, length=1, radius=0.5, shapeName=dyn_top_locane_t_r_02_capsule, rotation=<rot 90 0 0>}]
 head [{
   shapeType=Capsule
   length=4.9584
   radius=4.9671
   shapeName=head_capsule
   center=<3.0664 1.3465 -0.088>
   rotation=<rot 1.1342 -90.0004 88.4458>
  }]
 neck_02 [{shapeType=Capsule, length=8.4877, radius=4.233, shapeName=neck_02_capsule, rotation=<rot 0 -90.0001 90>}]
}
constraints [
 {
  bone1Name=dyn_ear_l_chain_02
  bone2Name=dyn_ear_l_chain_01
  swing1LimitDegrees=15
  swing2LimitDegrees=35
  twistLimitDegrees=45
  twistMotion=Locked
 }
 {
  bone1Name=dyn_ear_l_chain_03
  bone2Name=dyn_ear_l_chain_02
  swing1LimitDegrees=15
  swing2LimitDegrees=35
  twistLimitDegrees=45
  twistMotion=Locked
 }
 {
  bone1Name=dyn_ear_r_chain_02
  bone2Name=dyn_ear_r_chain_01
  swing1LimitDegrees=15
  swing2LimitDegrees=35
  twistLimitDegrees=45
  twistMotion=Locked
 }
 {
  bone1Name=dyn_ear_r_chain_03
  bone2Name=dyn_ear_r_chain_02
  swing1LimitDegrees=15
  swing2LimitDegrees=35
  twistLimitDegrees=45
  twistMotion=Locked
 }
 {bone1Name=dyn_front_locane_l_02, bone2Name=dyn_front_locane_l_01, swing1LimitDegrees=25, swing2LimitDegrees=25, twistLimitDegrees=0}
 {bone1Name=dyn_front_locane_r_02, bone2Name=dyn_front_locane_r_01, swing1LimitDegrees=25, swing2LimitDegrees=25, twistLimitDegrees=0}
 {bone1Name=dyn_top_locane_d_l_02, bone2Name=dyn_top_locane_d_l_01, swing1LimitDegrees=5, swing2LimitDegrees=35, twistLimitDegrees=0}
 {bone1Name=dyn_top_locane_d_r_02, bone2Name=dyn_top_locane_d_r_01, swing1LimitDegrees=69.72, swing2LimitDegrees=15, twistLimitDegrees=0}
 {bone1Name=dyn_top_locane_t_l_02, bone2Name=dyn_top_locane_t_l_01, swing1LimitDegrees=15, swing2LimitDegrees=15, twistLimitDegrees=0}
 {bone1Name=dyn_top_locane_t_l_03, bone2Name=dyn_top_locane_t_l_02, swing1LimitDegrees=15, swing2LimitDegrees=15, twistLimitDegrees=0}
 {bone1Name=dyn_top_locane_t_r_02, bone2Name=dyn_top_locane_t_r_01, swing1LimitDegrees=46.08, swing2LimitDegrees=15, twistLimitDegrees=0}
 {bone1Name=dyn_ear_r_chain_01, bone2Name=head, swing1LimitDegrees=35, swing2LimitDegrees=35, twistLimitDegrees=45, twistMotion=Locked}
 {bone1Name=dyn_ear_l_chain_01, bone2Name=head, swing1LimitDegrees=35, swing2LimitDegrees=35, twistLimitDegrees=45, twistMotion=Locked}
 {bone1Name=dyn_front_locane_r_01, bone2Name=head, swing1LimitDegrees=25, swing2LimitDegrees=25, twistLimitDegrees=0}
 {bone1Name=dyn_top_locane_d_r_01, bone2Name=head, swing1LimitDegrees=82.68, swing2LimitDegrees=25.08, twistLimitDegrees=0}
 {bone1Name=dyn_top_locane_t_r_01, bone2Name=head, swing1LimitDegrees=26.52, swing2LimitDegrees=13.56, twistLimitDegrees=0}
 {bone1Name=dyn_top_locane_t_l_01, bone2Name=head, swing1LimitDegrees=15, swing2LimitDegrees=15, twistLimitDegrees=0}
 {bone1Name=dyn_top_locane_d_l_01, bone2Name=head, swing1LimitDegrees=5, swing2LimitDegrees=59.48, twistLimitDegrees=0}
 {bone1Name=dyn_front_locane_l_01, bone2Name=head, swing1LimitDegrees=25, swing2LimitDegrees=25, twistLimitDegrees=0}
 {
  bone1Name=head
  bone2Name=neck_02
  swing1LimitDegrees=45
  swing1Motion=Free
  swing2LimitDegrees=45
  swing2Motion=Free
  twistLimitDegrees=45
  twistMotion=Free
 }
]
